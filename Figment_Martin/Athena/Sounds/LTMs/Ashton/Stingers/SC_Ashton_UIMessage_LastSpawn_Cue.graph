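# Sound cue for the "last spawn" UI message stinger in the Ashton limited-time mode.
# A single wave player feeds the Ashton_UIMessage_LastSpawn wave straight to the output, with no randomization or modulation.
# It routes through the important in-game mission sound class (Fort_Mission_Ingame_Important_PSM).
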SoundCue /Figment_Martin/Athena/Sounds/LTMs/Ashton/Stingers/SC_Ashton_UIMessage_LastSpawn_Cue
root=WavePlayer_0
nodes [{id=WavePlayer_0, properties {soundWaveAssetPtr=&/Ashton_UIMessage_LastSpawn}}]
properties {maxDistance=2097152, soundClassObject=&Fort_Mission_Ingame_Important_PSM}
